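# Additive unlit particle material for a Cloak Clean Sweep effect, drawing a noisy glowing disc or ring on Niagara sprites, named as a particle decal.
# Two noise-distorted sphere masks are subtracted, with dynamic parameters controlling the inner cutout's size and strength.
# A Fresnel from flattened scene world normals shades it, with constant fallbacks on forward and mobile.
# Depth-based distance masking and particle color feed a shared emissive adjustment function.
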Material /Game/Effects/Fort_Effects/Materials/ParticleDecals/M_Cloak_CleanSweep_Sphere_PD
graph {
 properties {
  bUsedWithNanite=true
  bUsedWithNiagaraSprites=true
  blendMode=BLEND_Additive
  pixelDepthOffsetMode=PDOM_Legacy
  shadingModel=MSM_Unlit
  usedShadingModels=Unlit
 }
 nodes [
  n1 = TextureCoordinate()
  n2 = Constant() {r=0.5}
  n3 = TextureCoordinate()
  n4 = Divide(A=ComponentMask(None=TextureSample(UVs=@n3) {samplerType=SAMPLERTYPE_LinearColor, texture=&T_NoisePack_LinearColor}.RGB) {g=true, r=true}) {constB=3}
  n5 = DynamicParameter() {paramNames [FresnelExponent, PanRandom, Param3, Param4]}
  n6 = SphereMask(A=@n1, B=@n2, Radius=LinearInterpolate(A=LinearInterpolate(Alpha=@n4) {constA=0.25, constB=0.4}, Alpha=@n5.Param4) {constB=0.5}) {attenuationRadius=0.45, hardnessPercent=0}
  n7 = Subtract(
   A=SphereMask(A=@n1, B=@n2, Radius=LinearInterpolate(Alpha=@n4) {constA=0.4, constB=0.5}) {attenuationRadius=0.5, hardnessPercent=5}
   B=Multiply(A=Multiply(A=@n6) {constB=1.2}, B=@n5.Param3)
  )
  n8 = SceneTexture() {sceneTextureId=PPI_WorldNormal}
  n9 = Fresnel(
   ExponentIn=Multiply(
    A=ScalarParameter() {defaultValue=1, group=Alpha, parameterName=FresnelExponent}
    B=DynamicParameter() {paramNames [FresnelExponent, PanRandom, Param3, Param4]}.FresnelExponent
   )
   Normal=MaterialFunctionCall(
    Normal=ComponentMask(None=@n8.Color) {b=true, g=true, r=true}
    Flatness=ScalarParameter() {defaultValue=0.2, group=Alpha, parameterName=FlattenNormal}
   ) {materialFunction=&FlattenNormal}.Result
  ) {baseReflectFraction=0, exponent=1}
  n10 = Constant3Vector() {constant=<rgba 0.5 0.5 0.5 1>}
  n11 = ShadingPathSwitch(Default=@n9, Deferred=@n9, Forward=@n10, Mobile=@n10)
  n12 = Multiply(
   A=Multiply(A=Power(Base=@n7)) {constB=10}
   B=ShadingPathSwitch(Default=@n11, Deferred=@n11, Mobile=Constant() {r=0.3}) {desc=Migrated from FeatureLevelSwitch}
  )
  n13 = MaterialFunctionCall!(
   Distance Bias=ScalarParameter() {defaultValue=20, parameterName=CameraOffsetBias}
   Distance Divisor=ScalarParameter() {defaultValue=60, parameterName=DistanceDivisor}
   Scene Depth
   Pixel Depth
  ) {materialFunction=/Game/Effects/Fort_Effects/Materials/_Functions/MF_DistanceBehindTranslucency_01}
  MaterialFunctionCall(Emissive=Multiply(A=Multiply(A=@n12, B=@n13.Inverted), B=ParticleColor().RGB)) {materialFunction=&MF_FX_EmissiveAdjustments}
 ]
}
