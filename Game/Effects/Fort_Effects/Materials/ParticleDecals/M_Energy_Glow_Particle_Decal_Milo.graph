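# Additive unlit particle sprite material for a decal-like energy glow that picks up the scene surface beneath it.
# Radially mapped, panning smoke texture, optionally posterized, mixes dark and particle colors, plus a Fresnel edge from flattened scene world normals.
# Sphere mask, Fresnel, depth fades and optional time-of-day scaling shape opacity; dynamic parameters vary Fresnel and panning.
# A camera-distance offset collapses nearby sprites toward their centers; mobile uses constant fallbacks and object position.
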Material /Game/Effects/Fort_Effects/Materials/ParticleDecals/M_Energy_Glow_Particle_Decal_Milo
graph {
 properties {bUsedWithParticleSprites=true, blendMode=BLEND_Additive, pixelDepthOffsetMode=PDOM_Legacy, shadingModel=MSM_Unlit, usedShadingModels=Unlit}
 nodes [
  n1 = ParticlePositionWS()
  n2 = ShadingPathSwitch(Default=@n1, Deferred=@n1, Mobile=ObjectPositionWS()) {desc="Please fix particle position on mobile :("}
  n3 = Distance(A=@n2, B=CameraPositionWS())
  n4 = ScalarParameter() {defaultValue=250, parameterName=NearFadeDistance}
  LinearInterpolate(A=Subtract(A=@n2, B=WorldPosition().XYZ), Alpha=Clamp(None=Divide(A=@n3, B=@n4))) {constB=0}
  n5 = DynamicParameter() {paramNames [FresnelExponent, Refraction, PanAdd, Random]}
  n6 = Multiply(A=Time(), B=LinearInterpolate(Alpha=@n5.Random) {constA=0.4, constB=1.4})
  n7 = Add(A=@n5.PanAdd, B=@n5.PanAdd)
  n8 = ScalarParameter() {defaultValue=0.05, parameterName=PannerSpeedX}
  n9 = ScalarParameter() {defaultValue=-0.5, parameterName=PannerSpeedY}
  n10 = TextureSampleParameter2D(UVs=Panner(
    Coordinate=MaterialFunctionCall() {materialFunction=&VectorToRadialValue}.Radial Coordinates
    Time=Add(A=@n6, B=@n7)
    Speed=AppendVector(A=@n8, B=@n9)
   ) {speedX=0.05, speedY=0.5}) {parameterName=TilingTexture, texture=&T_Inky_Smoke_Tile_Chromatic}
  n11 = ScalarParameter() {defaultValue=5, parameterName=Floor}
  n12 = StaticSwitchParameter(True=Floor(None=Multiply(A=@n10.RGB, B=@n11)), False=@n10.RGB) {parameterName=Floor}
  n13 = ScalarParameter() {defaultValue=-1.5, parameterName=Desaturation Factor}
  n14 = Add(
   A=VectorParameter() {parameterName=DarkColor, defaultValue=<rgba 0.0501 0.0332 0.1302 1>}.RGB
   B=Multiply(A=Desaturation(None=@n12, Fraction=@n13)) {constB=1.5}
  )
  n15 = ParticleColor()
  n16 = LinearInterpolate(A=@n15.RGB, Alpha=ScalarParameter() {defaultValue=0.6, parameterName=Lack of Particle Influence}) {constAlpha=0.6}
  n17 = SceneTexture() {sceneTextureId=PPI_WorldNormal}
  n18 = Fresnel(
   ExponentIn=@n5.FresnelExponent
   Normal=MaterialFunctionCall(Normal=ComponentMask(None=@n17.Color) {b=true, g=true, r=true}, Flatness=Constant() {r=0.2}) {materialFunction=&FlattenNormal}.Result
  ) {baseReflectFraction=0, exponent=1}
  n19 = Constant() {r=0.3}
  n20 = ShadingPathSwitch(Default=ShadingPathSwitch(Default=@n18, Deferred=@n18, Forward=@n19, Mobile=@n18), Mobile=@n19) {desc=Migrated from FeatureLevelSwitch}
  n21 = Multiply(A=@n15.RGB, B=@n20)
  n22 = ScalarParameter() {defaultValue=5, parameterName=WorldNormalIntensity}
  MaterialFunctionCall(Emissive=Add(A=Multiply(A=@n14, B=@n16), B=Multiply(A=@n21, B=@n22) {constB=5, desc=Fresnel edge color})) {materialFunction=&MF_FX_EmissiveAdjustments}
  n23 = ScalarParameter() {parameterName=OpacityMorning}
  n24 = ScalarParameter() {parameterName=OpacityDay}
  n25 = ScalarParameter() {defaultValue=1, parameterName=OpacityEvening}
  n26 = ScalarParameter() {defaultValue=1, parameterName=OpacityNight}
  n27 = Constant() {r=0.5}
  n28 = SphereMask(A=TextureCoordinate(), B=@n27, Radius=@n27) {hardnessPercent=0}
  n29 = ScalarParameter() {defaultValue=2, parameterName=SphereMaskPower}
  n30 = ScalarParameter() {defaultValue=1, parameterName=TextureOpacityAdd}
  n31 = DepthFade(FadeDistance=ScalarParameter() {defaultValue=44, parameterName=VisibleDepthDistance}) {fadeDistanceDefault=80}
  n32 = DepthFade(Opacity=Add(A=Multiply(A=@n12, B=@n30), B=OneMinus(None=@n31)), FadeDistance=ScalarParameter() {defaultValue=20, parameterName=DepthFade}) {fadeDistanceDefault=15}
  n33 = Multiply(
   A=Multiply(A=Multiply(A=@n15.A, B=@n20) {constA=1}, B=Power(Base=@n28, Exp=@n29))
   B=ShadingPathSwitch(Default=@n32, Deferred=@n32, Mobile=Constant() {r=0.75}) {desc=Migrated from FeatureLevelSwitch}
  )
  StaticSwitchParameter(
   True=Multiply(
    A=MaterialFunctionCall(Color_Morning=@n23, Color_Day=@n24, Color_Evening=@n25, Color_Night=@n26) {materialFunction=&Color_Time_of_Day}.Result
    B=@n33
   )
   False=@n33
  ) {parameterName=UseTimeofDayOnOpacity}
 ]
}
