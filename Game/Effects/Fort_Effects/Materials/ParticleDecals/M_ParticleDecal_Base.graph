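# Translucent unlit base material for particle decals on Niagara and sprite particles, projecting a texture onto the scene behind each particle.
# Scene position relative to the particle, scaled by size and transformed to tangent space, gives optionally panned UVs for blended mask or SubUV textures.
# Particle color drives emissive; the mask, with optional erosion, distance and box-gradient fades, drives opacity, adding depth fade on higher non-mobile quality.
# A WPO term pulls particles toward the camera.
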Material /Game/Effects/Fort_Effects/Materials/ParticleDecals/M_ParticleDecal_Base
graph {
 properties {
  bUsedWithNiagaraMeshParticles=true
  bUsedWithNiagaraSprites=true
  bUsedWithParticleSprites=true
  blendMode=BLEND_Translucent
  pixelDepthOffsetMode=PDOM_Legacy
  shadingModel=MSM_Unlit
  usedShadingModels=Unlit
 }
 nodes [
  n1 = Time()
  n2 = ScalarParameter() {defaultValue=1, group=Panner, parameterName=Panner Speed Scale, sortPriority=1}
  n3 = ScalarParameter() {group=Panner, parameterName=Panner U Speed, sortPriority=2}
  n4 = ScalarParameter() {group=Panner, parameterName=Panner V Speed, sortPriority=3}
  n5 = MaterialFunctionCall() {materialFunction=&WorldPositionBehindTranslucency}
  n6 = ParticlePositionWS()
  n7 = ParticleSize()
  n8 = Transform(None=Divide(A=Subtract(A=@n5.Result, B=@n6), B=ComponentMask(None=@n7) {r=true})) {transformSourceType=TRANSFORMSOURCE_World, transformType=TRANSFORM_Tangent}
  n9 = ScalarParameter() {defaultValue=1, group=WorldScale, parameterName=World Scale U}
  n10 = ScalarParameter() {defaultValue=1, group=WorldScale, parameterName=World Scale V, sortPriority=1}
  n11 = Subtract(A=Multiply(A=ComponentMask(None=@n8) {g=true, r=true}, B=AppendVector(A=@n9, B=@n10)), B=Constant2Vector() {g=0.5, r=0.5})
  n12 = StaticSwitchParameter(True=Panner(Coordinate=@n11, Time=Multiply(A=@n1, B=@n2), Speed=AppendVector(A=@n3, B=@n4)), False=@n11) {group=Panner, parameterName=Use Panner}
  n13 = TextureSampleParameterSubUV!(UVs=@n12, Apply View MipBias) {group=Textures, parameterName=SubUVTex, texture=/Game/Effects/Fort_Effects/Gameplay/SmokeDustExplosions/Textures/T_Explosion_Streak_02}
  n14 = LinearInterpolate(
   A=TextureSampleParameter2D(UVs=@n12) {parameterName=Mask Tex, texture=/Game/Effects/Fort_Effects/Materials/T_EFX_Flare_Burst_01}.RGBA
   B=@n13.RGBA
   Alpha=ScalarParameter() {group=Panner, parameterName=UseSubUVTex, sortPriority=3}
  )
  AppendVector(A=ComponentMask(None=@n14) {b=true, g=true, r=true}, B=ComponentMask(None=@n14) {a=true})
  n15 = ParticleColor()
  MaterialFunctionCall(Emissive=@n15.RGB) {materialFunction=&MF_FX_EmissiveAdjustments}
  n16 = ParticlePositionWS()
  n17 = CameraPositionWS()
  n18 = ParticleSize()
  Multiply(A=Normalize(VectorInput=Subtract(A=@n16, B=@n17)), B=Multiply(A=ComponentMask(None=@n18) {r=true}) {constB=-0.5}) {desc=WPO input}
  n19 = LinearInterpolate(
   A=@n14
   B=Subtract(A=@n14, B=DynamicParameter() {defaultValue=<rgba 0 1 1 1>, paramNames [Erosion,Param2,Param3,Param4]}.Erosion) {constB=0.5}
   Alpha=ScalarParameter() {group=Panner, parameterName=UseErosion, sortPriority=3}
  )
  n20 = Divide(
   A=Distance(
    A=MaterialFunctionCall() {materialFunction=&WorldPositionBehindTranslucency}.Result
    B=ParticlePositionWS()
   )
   B=ScalarParameter() {defaultValue=5, group=Distance Fade, parameterName=Fade Distance}
  )
  n21 = ScalarParameter() {defaultValue=5, group=Distance Fade, parameterName=Fade Start Distance, sortPriority=1}
  n22 = OneMinus(None=Clamp(None=Subtract(A=@n20, B=@n21)))
  n23 = TextureCoordinate()
  n24 = ComponentMask(None=TextureCoordinate()) {r=true}
  n25 = Multiply(
   A=Power(
    Base=ComponentMask(None=@n23) {r=true}
    Exp=ScalarParameter() {defaultValue=1, group=Opacity, parameterName=Left-Right Contrast, sortPriority=2}
   )
   B=Power(Base=OneMinus(None=@n24), Exp=ScalarParameter() {defaultValue=1, group=Opacity, parameterName=Right-Left Contrast, sortPriority=3})
  )
  n26 = TextureCoordinate()
  n27 = ComponentMask(None=TextureCoordinate()) {g=true}
  n28 = Multiply(
   A=Power(
    Base=ComponentMask(None=@n26) {g=true}
    Exp=ScalarParameter() {defaultValue=1, group=Opacity, parameterName=Top-Bottom Contrast, sortPriority=4}
   )
   B=Power(Base=OneMinus(None=@n27), Exp=ScalarParameter() {defaultValue=1, group=Opacity, parameterName=Bottom-Top Contrast, sortPriority=5})
  )
  n29 = Clamp(None=Multiply(
    A=Multiply(A=@n25, B=@n28)
    B=ScalarParameter() {defaultValue=100, group=Opacity, parameterName=Opacity_Intensity_Compensator, sortPriority=1}
   ))
  n30 = Constant() {r=1}
  n31 = Multiply(
   A=Multiply(A=@n15.A, B=StaticComponentMaskParameter(None=@n19) {defaultG=true, parameterName=Mask channel})
   B=Multiply(A=Multiply(A=@n22, B=@n22), B=StaticSwitchParameter(True=@n29, False=@n30) {group=Opacity, parameterName=Use Box Gradient Opacity Control})
  )
  n32 = DepthFade(Opacity=@n31, FadeDistance=ScalarParameter() {parameterName=DepthFadeDistance})
  QualitySwitch(Default=ShadingPathSwitch(Default=@n32, Mobile=@n31), Low=@n31, High=@n32, Medium=@n31, Epic=@n32)
 ]
}
