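# Unlit additive particle material for sprite and Niagara particles that fakes a spotlight pool of light without textures.
# Particle color feeds an emissive adjustment function, while particle alpha scales a procedural mask made from a SphereMask circle minus an offset shadow circle.
# Nested DepthFades limit the mask by scene depth, and a Fresnel on the flattened SceneTexture world normal shades it by surface orientation.
# Shading path switches substitute constants for these scene-based terms on mobile, and for the Fresnel on forward.
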Material /Game/Effects/Fort_Effects/Materials/ParticleDecals/M_ParticleDecal_FakeSpotlight
graph {
 properties {
  bUsedWithNiagaraSprites=true
  bUsedWithParticleSprites=true
  blendMode=BLEND_Additive
  pixelDepthOffsetMode=PDOM_Legacy
  shadingModel=MSM_Unlit
  usedShadingModels=Unlit
 }
 nodes [
  n1 = ParticleColor()
  MaterialFunctionCall(Emissive=@n1.RGB) {materialFunction=&MF_FX_EmissiveAdjustments}
  n2 = VectorParameter() {parameterName=SphereMask Placement, defaultValue=<rgba 0.5 0.5 0 0>}
  n3 = Power(
   Base=SphereMask(
    A=TextureCoordinate()
    B=ComponentMask(None=@n2.RGB) {g=true, r=true}
    Radius=ScalarParameter() {defaultValue=0.5, parameterName=Radius}
    Hardness=ScalarParameter() {defaultValue=0.619, parameterName=Hardness}
   )
   Exp=ScalarParameter() {defaultValue=1, parameterName=Opacity_Contrast}
  )
  n4 = TextureCoordinate()
  n5 = ComponentMask(None=VectorParameter() {parameterName=SphereMask Shadow_UVAdjustments, defaultValue=<rgba 0 -0.1048 0 0>}.RGB) {g=true, r=true}
  n6 = VectorParameter() {parameterName=SphereMask ShadowPlacement, defaultValue=<rgba 0.5 0.5 0 0>}
  n7 = Power(
   Base=SphereMask(
    A=Add(A=@n4, B=@n5)
    B=ComponentMask(None=@n6.RGB) {g=true, r=true}
    Radius=ScalarParameter() {defaultValue=0.1667, parameterName=Shadow_Radius}
    Hardness=ScalarParameter() {parameterName=Shadow_Hardness}
   )
   Exp=ScalarParameter() {defaultValue=1, parameterName=Shadow_Contrast}
  )
  n8 = ScalarParameter() {defaultValue=1500, parameterName=VisibleDepthDistance}
  n9 = DepthFade(
   Opacity=OneMinus(None=DepthFade(FadeDistance=@n8) {fadeDistanceDefault=80})
   FadeDistance=ScalarParameter() {defaultValue=500, parameterName=DepthFade}
  ) {fadeDistanceDefault=15}
  n10 = ShadingPathSwitch(Default=@n9, Deferred=@n9, Mobile=Constant() {r=0.75}) {desc=Migrated from FeatureLevelSwitch}
  n11 = SceneTexture() {sceneTextureId=PPI_WorldNormal}
  n12 = Fresnel(Normal=MaterialFunctionCall(Normal=ComponentMask(None=@n11.Color) {b=true, g=true, r=true}, Flatness=Constant() {r=0.2}) {materialFunction=&FlattenNormal}.Result) {baseReflectFraction=0, exponent=1}
  n13 = Constant() {r=0.3}
  n14 = ShadingPathSwitch(Default=ShadingPathSwitch(Default=@n12, Deferred=@n12, Forward=@n13, Mobile=@n12), Mobile=@n13) {desc=Migrated from FeatureLevelSwitch}
  Multiply(A=@n1.A, B=Multiply(A=Subtract(A=@n3, B=@n7), B=Multiply(A=@n10, B=@n14)) {constA=1})
 ]
}
