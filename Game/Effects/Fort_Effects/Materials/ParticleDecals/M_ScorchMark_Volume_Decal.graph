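# Unlit translucent deferred decal material for Niagara-spawned scorch marks with an animated emissive center.
# Radial masks from world position relative to the actor combine with sine-warped volume noise, avoiding flat projected UVs.
# DecalColor RGB scales mask range rather than tinting, while alpha acts as second opacity.
# Lifetime opacity drives emissive and optional inverted erosion; quality switches limit extra emissive detail and noise-derivative normals to higher tiers.
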Material /Game/Effects/Fort_Effects/Materials/ParticleDecals/M_ScorchMark_Volume_Decal
graph {
 properties {
  bAutomaticallySetUsageInEditor=false
  blendMode=BLEND_Translucent
  materialDomain=MD_DeferredDecal
  pixelDepthOffsetMode=PDOM_Legacy
  shadingModel=MSM_Unlit
  usedShadingModels=Unlit
 }
 nodes [
  Multiply(A=NamedRerouteUsage(), B=VectorParameter() {group=Base, parameterName=BaseColor, defaultValue=<rgba 0.0155 0.0224 0.0312 1>}.RGB)
  n1 = Multiply(A=NamedRerouteUsage(), B=ScalarParameter() {defaultValue=0.0017, group=Sine, parameterName=Volume Sphere Center Mask Size})
  n2 = Power(Base=OneMinus(None=Saturate(None=@n1)), Exp=ScalarParameter() {defaultValue=9.2, group=Sine, parameterName="Mask Center Gradient  Pow"})
  n3 = ScalarParameter() {defaultValue=400, group=Sine, parameterName=Gradient Center Intensity}
  n4 = NamedRerouteUsage()
  n5 = WorldPosition()
  n6 = Subtract(A=ActorPositionWS(), B=@n5.XYZ)
  n7 = Constant3Vector()
  n8 = Distance(A=@n6, B=@n7)
  n9 = Multiply(A=@n8, B=ScalarParameter() {defaultValue=0.001, parameterName=Distance Adjustment}) {constB=0.01}
  n10 = Multiply(A=Add(A=Sine(None=@n9))) {constB=0.5}
  n11 = ActorPositionWS()
  n12 = ScalarParameter() {defaultValue=0.004, group=Volume Noise, parameterName=UVW_Scale}
  n13 = Multiply(A=Multiply(A=Power(Base=@n10) {constExponent=-4}, B=Subtract(A=@n11, B=@n5.XYZ)), B=@n12)
  n14 = LinearInterpolate(
   A=NamedRerouteDeclaration(None=TextureSample(UVs=@n13) {samplerType=SAMPLERTYPE_LinearColor, texture=/Game/Global/MaterialFunctions/VT_ClothSimWPO}.R) {name=Noise2, nodeColor=<rgba 0 0.625 1 1>}
   B=TextureSample(UVs=Multiply(A=@n12, B=@n5.XYZ)) {samplerType=SAMPLERTYPE_LinearColor, texture=/Game/Effects/Explosion_Update/Textures/VT_Voronoi_Seamless}.R
  )
  n15 = ScalarParameter() {defaultValue=2, group=Volume Noise, parameterName=NoisePow}
  n16 = Saturate(None=Add(A=Multiply(A=@n4) {constB=0.5}, B=Power(Base=@n14, Exp=@n15)))
  NamedRerouteDeclaration(None=Multiply(A=Multiply(A=@n2, B=@n3), B=@n16)) {name=Sine Mask Main_1, nodeColor=<rgba 1 0 0.9375 1>}
  n17 = Multiply(A=ScalarParameter() {defaultValue=0.002, group=Mask, parameterName=CenterMask Adjustment}, B=@n8) {constB=0.05}
  n18 = OneMinus(None=Saturate(None=Multiply(A=@n17) {constB=5}))
  NamedRerouteDeclaration(None=Multiply(A=@n18, B=@n18)) {name=Lightning Mask, nodeColor=<rgba 0.7578 1 0 1>}
  n19 = Multiply(A=@n5.XYZ, B=ScalarParameter() {defaultValue=0.004, group=Noise, parameterName=Noise Distortion UV Scale})
  n20 = Multiply(A=Time(), B=ScalarParameter() {defaultValue=0.04, parameterName=Noise UV Anim Speed})
  NamedRerouteDeclaration(None=TextureSample(UVs=Add(A=@n19, B=@n20)) {samplerType=SAMPLERTYPE_LinearColor, texture=&VT_ClothSimWPO}.R) {name=Noise, nodeColor=<rgba 1 0.2344 0 1>}
  n21 = DecalColor() {desc="RGB= Result of Decal Range Adjustment via Niagara\r\n!! It's NOT Color value in this Material.\r\n\r\nA= Second Opacity"}
  NamedRerouteDeclaration(None=Distance(A=Multiply(A=@n21.RGB, B=@n6), B=@n7)) {name=Distance, nodeColor=<rgba 0.6875 0 1 1>}
  n22 = NamedRerouteUsage()
  n23 = Multiply(A=@n22, B=@n22)
  NamedRerouteDeclaration(None=@n23) {name=Mask2, nodeColor=<rgba 0.4766 1 0 1>}
  n24 = Multiply(A=ScalarParameter() {defaultValue=0.02, group=Volume Noise, parameterName=UVW_Scale_High}, B=@n5.XYZ)
  n25 = Multiply(A=Time()) {constB=-0.126}
  n26 = LinearInterpolate(
   A=NamedRerouteUsage()
   B=TextureSample(UVs=Add(A=@n24, B=@n25)) {samplerType=SAMPLERTYPE_LinearColor, texture=&VT_Voronoi_Seamless}.R
  ) {constAlpha=0.35}
  n27 = Multiply(A=@n26, B=@n26)
  NamedRerouteDeclaration(None=Multiply(A=@n27, B=@n27)) {name=Emissive Detail, nodeColor=<rgba 0 1 0.7422 1>}
  n28 = Constant3Vector() {constant=<rgba 0 0 1 0>}
  n29 = MaterialFunctionCall(R=DDX(Value=@n16), G=DDY(Value=@n16), B=Constant() {r=1}) {materialFunction=&AppendMany}
  NamedRerouteDeclaration(None=MaterialFunctionCall!(
    Low _ PC Low and Med=@n29.RGB
    "High _ XB1, PS4"=@n29.RGB
    "Epic _ PC High, XSX, PS5"=@n29.RGB
    --------=@n28
    "Mobile Low _ PC Perf, Android"=@n28
    Mobile Med _ Switch=@n28
   ) {materialFunction=/Game/Global/MaterialFunctions/TechArt/MF_Fortnite_QualitySwitch_Vector3}.Result) {name=Normal Result, nodeColor=<rgba 1 0 0.7734 1>}
  NamedRerouteUsage()
  ScalarParameter() {defaultValue=0.05, parameterName=aaa}
  ScalarParameter() {defaultValue=0.5, group=Base, parameterName=Metallic}
  ScalarParameter() {group=Base, parameterName=Specular}
  ScalarParameter() {defaultValue=0.5, group=Base, parameterName=Roughness}
  n30 = OneMinus(None=DecalLifetimeOpacity().Opacity)
  n31 = Subtract(A=Multiply(A=@n30, B=@n30)) {constB=0.77}
  n32 = VectorParameter() {group=Emissive, parameterName=Center Noise Color, defaultValue=<rgba 0 5.7024 30 1>}
  n33 = Multiply(A=Multiply(A=Multiply(A=@n31, B=@n23), B=NamedRerouteUsage()), B=@n32.RGB)
  n34 = Power(Base=@n23, Exp=ScalarParameter() {defaultValue=1, parameterName=Additional Mask Pow})
  n35 = Multiply(A=NamedRerouteUsage(), B=ScalarParameter() {defaultValue=2, group=Emissive, parameterName=Additional Emissive Detail Intensity}) {constB=2}
  n36 = Add(A=@n33, B=Multiply(A=@n32.RGB, B=Multiply(A=@n34, B=@n35)))
  StaticSwitchParameter(
   True=MaterialFunctionCall(
    Low _ PC Low and Med=@n33
    "High _ XB1, PS4"=@n36
    "Epic _ PC High, XSX, PS5"=@n36
    --------=@n33
    "Mobile Low _ PC Perf, Android"=@n33
    Mobile Med _ Switch=@n33
   ) {materialFunction=&MF_Fortnite_QualitySwitch_Vector3}.Result
   False=Constant()
  ) {defaultValue=true, parameterName=USE_Emissive}
  n37 = Multiply(A=NamedRerouteUsage(), B=ScalarParameter() {defaultValue=15, group=Inverted Erosion, parameterName=Erosion Intensity}) {constB=15}
  n38 = DecalLifetimeOpacity()
  n39 = StaticSwitchParameter(True=Add(A=Multiply(A=@n37, B=@n38.Opacity), B=@n38.Opacity), False=@n38.Opacity) {group=Inverted Erosion, parameterName=Inverted Erosion}
  n40 = NamedRerouteUsage()
  Subtract(
   A=Multiply(A=DecalColor() {desc="A= Second Opacity"}.A, B=NamedRerouteUsage())
   B=Multiply(A=Multiply(A=@n39, B=@n40), B=ScalarParameter() {defaultValue=5, group=Mask, parameterName=Scorch Opacity Adjustment}) {constB=50}
  )
 ]
}
