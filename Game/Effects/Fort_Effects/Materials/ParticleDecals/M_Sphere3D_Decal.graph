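# Translucent deferred decal material that draws a noisy spherical ramp on surfaces, based on world-space distance from the decal's position.
# A 3D noise texture's green channel varies the radius; its red channel feeds a smoothstep dissolve.
# One ramp lookup shifts by offset and camera-space scene normal, supplying color and alpha tinted by decal color.
# An unshifted lookup supplies a static alpha mask that multiplies the opacity.
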Material /Game/Effects/Fort_Effects/Materials/ParticleDecals/M_Sphere3D_Decal
graph {
 properties {bAutomaticallySetUsageInEditor=false, blendMode=BLEND_Translucent, materialDomain=MD_DeferredDecal, usedShadingModels=Default Lit}
 nodes [
  DecalColor()
  DecalLifetimeOpacity()
  DecalLifetimeOpacity()
  n1 = WorldPosition()
  n2 = ObjectPositionWS()
  n3 = ScalarParameter() {defaultValue=780.8436, parameterName=RadiusWS}
  n4 = ScalarParameter() {defaultValue=200, parameterName=RadiusRandom}
  n5 = TextureSampleParameterVolume!(UVs=Divide(A=WorldPosition().XYZ, B=ScalarParameter() {defaultValue=512, parameterName=DetailTextureUvwScaleWS}), Apply View MipBias) {parameterName="DetailTexture3D(Dissolve RRandom X X)", texture=/Game/Effects/Explosion_Update/Textures/VT_ExplosionNoise}
  n6 = Divide(
   A=Length(None=Subtract(A=@n1.XYZ, B=@n2)) {is=class MaterialExpressionLength}
   B=LinearInterpolate(A=Add(A=@n3, B=@n4), B=Subtract(A=@n3, B=@n4), Alpha=@n5.G)
  )
  n7 = ScalarParameter() {parameterName=Ramp_Offset}
  n8 = Multiply(
   A=Transform(None=SceneTexture() {sceneTextureId=PPI_WorldNormal}.Color) {transformSourceType=TRANSFORMSOURCE_World, transformType=TRANSFORM_Camera}
   B=ScalarParameter() {defaultValue=0.1}
  )
  n9 = TextureSampleParameter2D(UVs=Add(A=Subtract(A=@n6, B=@n7), B=ComponentMask(None=@n8) {g=true, r=true})) {
   parameterName="Ramp(X StaticAlpha Emi Alpha)"
   samplerSource=SSM_Clamp_WorldGroupSettings
   texture=/Game/Effects/Fort_Effects/Textures/Ramps/T_Power_Source_Sphere_Ramp_Soft
  }
  n10 = DecalColor()
  Multiply(A=@n9.B, B=@n10.RGB)
  n11 = TextureSampleParameter2D(UVs=@n6) {
   parameterName="Ramp(X StaticAlpha Emi Alpha)"
   samplerSource=SSM_Clamp_WorldGroupSettings
   texture=&T_Power_Source_Sphere_Ramp_Soft
  }
  Multiply(A=Multiply(A=Multiply(A=@n9.A, B=@n10.A), B=@n11.G), B=NamedRerouteUsage())
  n12 = NamedRerouteUsage()
  n13 = ScalarParameter() {group=Dissolve, parameterName=DissolveOffset}
  n14 = Multiply(A=ScalarParameter() {defaultValue=0.3915, group=Dissolve, parameterName=DissolveMin}, B=OneMinus(None=Multiply(A=@n12, B=@n13)))
  n15 = ScalarParameter() {defaultValue=0.4267, group=Dissolve, parameterName=DissolveMax}
  NamedRerouteDeclaration(None=SmoothStep(Min=@n14, Max=Add(A=@n14, B=@n15), Value=@n12)) {name=DissolveResault, nodeColor=<rgba 1 0.9405 0.1679 1>}
  NamedRerouteDeclaration(None=@n5.R) {name=Alpha, nodeColor=<rgba 1 0.1875 0 1>}
  VectorParameter() {parameterName=BaseColor}
 ]
}
